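# Anim Blueprint for the Contaminate medium male body's face accessory, which poses its antennas when the character is targeting or emoting.
# It copies the attached parent's pose, then applies additive bone offsets to both three-segment antennas.
# An eased targeting value drives their alphas through staggered ranges, so the tips follow the base segments.
# The head bone gets an additive counter-rotation of 20% of its ref-pose delta relative to neck_02.
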AnimBlueprint /BRCosmetics/Characters/Player/Male/Medium/Bodies/M_MED_Contaminate/Meshes/Parts/M_MED_Contaminate_FaceAcc_AnimBP
target_skeleton=&/M_MED_Contaminate_FaceAcc_Skeleton
graphs [
 AnimGraph {nodes [
   n1 = Get targeting() -> (targeting:double)
   n2 = "Transform (Modify) Bone - Bone: antenna_01_l"!(ComponentPose=Local To Component(LocalPose=Copy Pose From Mesh() {bUseAttachedParent=true}), Alpha=@n1, Translation="3.25,1,-2") -> (Pose) {
    translationMode=BMM_Additive
    translationSpace=BCS_BoneSpace
    alphaBoolBlend {blendOption=Linear}
    alphaScaleBiasClamp {bMapRange=true, inRange {max=6}}
    boneToModify {boneName=antenna_01_l}
    translation=<3.25 1 -2>
   }
   n3 = "Transform (Modify) Bone - Bone: antenna_01_l"!(ComponentPose=@n2, Alpha=@n1, Rotation:Rotator="45,-25,0") -> (Pose) {
    rotationMode=BMM_Additive
    rotationSpace=BCS_BoneSpace
    translationSpace=BCS_BoneSpace
    alphaBoolBlend {blendOption=Linear}
    alphaScaleBiasClamp {bMapRange=true, inRange {max=6}}
    boneToModify {boneName=antenna_01_l}
    rotation {pitch=45, yaw=-25}
   }
   n4 = "Transform (Modify) Bone - Bone: antenna_02_l"!(ComponentPose=@n3, Alpha=@n1, Rotation="-45,0,0") -> (Pose) {
    rotationMode=BMM_Additive
    rotationSpace=BCS_BoneSpace
    translationSpace=BCS_BoneSpace
    alphaBoolBlend {blendOption=Linear}
    alphaScaleBiasClamp {bMapRange=true, inRange {max=6}}
    boneToModify {boneName=antenna_02_l}
    rotation {pitch=-45}
   }
   n5 = "Transform (Modify) Bone - Bone: antenna_03_l"!(ComponentPose=@n4, Alpha=@n1, Translation="1.25,-0.53,0.735", Rotation="1.38,0,0") -> (Pose) {
    rotationMode=BMM_Additive
    rotationSpace=BCS_BoneSpace
    translationMode=BMM_Additive
    translationSpace=BCS_BoneSpace
    alphaBoolBlend {blendOption=Linear}
    alphaScaleBiasClamp {bMapRange=true, inRange {max=10, min=4}}
    boneToModify {boneName=antenna_03_l}
    rotation {pitch=1.38}
    translation=<1.25 -0.53 0.735>
   }
   n6 = "Transform (Modify) Bone - Bone: antenna_01_r"!(ComponentPose=@n5, Alpha=@n1, Translation="-3.25,-1,2") -> (Pose) {
    translationMode=BMM_Additive
    translationSpace=BCS_BoneSpace
    alphaBoolBlend {blendOption=Linear}
    alphaScaleBiasClamp {bMapRange=true, inRange {max=6}}
    boneToModify {boneName=antenna_01_r}
    translation=<-3.25 -1 2>
   }
   n7 = "Transform (Modify) Bone - Bone: antenna_01_r"!(ComponentPose=@n6, Alpha=@n1, Rotation:Rotator="45,-25,0") -> (Pose) {
    rotationMode=BMM_Additive
    rotationSpace=BCS_BoneSpace
    translationSpace=BCS_BoneSpace
    alphaBoolBlend {blendOption=Linear}
    alphaScaleBiasClamp {bMapRange=true, inRange {max=6}}
    boneToModify {boneName=antenna_01_r}
    rotation {pitch=45, yaw=-25}
   }
   n8 = "Transform (Modify) Bone - Bone: antenna_02_r"!(ComponentPose=@n7, Alpha=@n1, Rotation="-45,0,0") -> (Pose) {
    rotationMode=BMM_Additive
    rotationSpace=BCS_BoneSpace
    translationSpace=BCS_BoneSpace
    alphaBoolBlend {blendOption=Linear}
    alphaScaleBiasClamp {bMapRange=true, inRange {max=6}}
    boneToModify {boneName=antenna_02_r}
    rotation {pitch=-45}
   }
   n9 = "Transform (Modify) Bone - Bone: antenna_03_r"!(ComponentPose=@n8, Alpha=@n1, Translation="-1.25,0.53,-0.735", Rotation="-1.38,0,0") -> (Pose) {
    rotationMode=BMM_Additive
    rotationSpace=BCS_BoneSpace
    translationMode=BMM_Additive
    translationSpace=BCS_BoneSpace
    alphaBoolBlend {blendOption=Linear}
    alphaScaleBiasClamp {bMapRange=true, inRange {max=10, min=4}}
    boneToModify {boneName=antenna_03_r}
    rotation {pitch=-1.38}
    translation=<-1.25 0.53 -0.735>
   }
   n10 = "Transform (Modify) Bone - Bone: head"(ComponentPose=@n9, Alpha=1, Translation="0,0,0", Rotation=Get head rot() -> (head rot:Rotator), Scale="1,1,1") {rotationMode=BMM_Additive, rotationSpace=BCS_BoneSpace, alphaBoolBlend {blendOption=Linear}, boneToModify {boneName=head}}
   Output Pose(Result=Component To Local(ComponentPose=@n10)) {layerGroup=None}
  ]}
 EventGraph {nodes [
   n1 = Event BlueprintUpdateAnimation()
   n2 = "To Float (Boolean)"(InBool=OR Boolean(A=Get bIsTargeting(), B=Get bIsPlayingEmote()))
   > Set targeting(targeting:double=FInterpTo(Current=Get targeting(), Target=float * float(A=@n2, B=10) {is=title K2Node_PromotableOperator}, DeltaTime=@n1.DeltaTimeX, InterpSpeed=8)) -> (Output_Get:double)
   n3 = GetOwnerSkeletalMeshComponent()
   n4 = BreakRotator(InRot=BreakTransform(InTransform=GetDeltaTransformFromRefPose(self=@n3, BoneName=head, BaseName=neck_02)).Rotation)
   > Set head rot(head rot:Rotator=MakeRotator(
     Roll=float * float!(A=@n4, B=-1, C:double=0.2) -> (ReturnValue) {is=title K2Node_CommutativeAssociativeBinaryOperator}
     Pitch=float * float(A=@n4.Pitch, B=-1, C=0.2)
     Yaw=float * float(A=@n4.Yaw, B=-1, C=0.2)
    )) -> (Output_Get:Rotator)
   TryGetPawnOwner()
   Get X percentage() -> (X percentage:float)
   Get Y percentage() -> (Y percentage:float)
   Get Z percentage() -> (Z percentage:float)
  ]}
]
variables [
 head rot {type=Rotator}
 targeting {type=number}
 X percentage {type=number}
 Y percentage {type=number}
 Z percentage {type=number}
]
